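# Sound cue for the Plum Kinetic Absorption explosion, played through the melee weapon sound class with its own attenuation asset.
# A distance crossfade blends a random pick of three close explosion waves into a random pick of four distant variants as the listener moves away, overlapping in mid range.
# A root modulator adds slight pitch variation, and a concurrency override caps simultaneous instances.
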SoundCue /HighTower/Audio/Plum/KineticAbsorption/PlumAbsorption_Explosion_Cue
root=Modulator_0
nodes [
 {
  id=DistanceCrossFade_0
  children [Random_0,Random_1]
  properties {crossFadeInput [
    {fadeInDistanceEnd=0, fadeInDistanceStart=0, fadeOutDistanceEnd=2500, fadeOutDistanceStart=1500, volume=1}
    {fadeInDistanceEnd=2200, fadeInDistanceStart=1300, fadeOutDistanceEnd=10000, fadeOutDistanceStart=8000, volume=1}
   ]}
 }
 {id=Modulator_0, children [DistanceCrossFade_0], properties {pitchMax=1.02, pitchMin=0.98}}
 {id=Random_0, children [WavePlayer_3,WavePlayer_2,WavePlayer_6], properties {weights [1,1,1]}}
 {id=Random_1, children [WavePlayer_4, WavePlayer_5, WavePlayer_8, WavePlayer_10], properties {weights [1,1,1,1]}}
 {id=WavePlayer_10, properties {soundWaveAssetPtr=&/PlumAbsorption_ExplosionDist_06}}
 {id=WavePlayer_2, properties {soundWaveAssetPtr=&/PlumAbsorption_Explosion_04}}
 {id=WavePlayer_3, properties {soundWaveAssetPtr=&/PlumAbsorption_Explosion_03}}
 {id=WavePlayer_4, properties {soundWaveAssetPtr=&/PlumAbsorption_ExplosionDist_02}}
 {id=WavePlayer_5, properties {soundWaveAssetPtr=&/PlumAbsorption_ExplosionDist_03}}
 {id=WavePlayer_6, properties {soundWaveAssetPtr=&/PlumAbsorption_Explosion_02}}
 {id=WavePlayer_8, properties {soundWaveAssetPtr=&/PlumAbsorption_ExplosionDist_04}}
]
properties {
 attenuationSettings=&/PlumAbsorption_Explosion_Att
 bOverrideConcurrency=true
 maxDistance=10000
 soundClassObject=&Fort_WeaponMelee
 volumeMultiplier=0.85
 concurrencyOverrides {maxCount=3}
}
