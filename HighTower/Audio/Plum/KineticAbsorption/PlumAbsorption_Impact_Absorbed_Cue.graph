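# Sound cue for the Plum Kinetic Absorption impact-absorbed sound, routed through the melee weapon sound class.
# A mixer layers a random pick of four equally weighted absorb waves with a fixed hit wave, slightly quieter and briefly delayed, then feeds a modulator.
# It uses the bullet impact attenuation and overrides concurrency to one instance per owner, stopping the oldest.
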SoundCue /HighTower/Audio/Plum/KineticAbsorption/PlumAbsorption_Impact_Absorbed_Cue
root=Modulator_0
nodes [
 {id=Delay_0, children [WavePlayer_0], properties {delayMax=0.1, delayMin=0.1}}
 {id=Mixer_0, children [Random_0,Delay_0], properties {inputVolume [1,0.85]}}
 {id=Modulator_0, children [Mixer_0]}
 {id=Random_0, children [WavePlayer_1, WavePlayer_7, WavePlayer_12, WavePlayer_13], properties {weights [1,1,1,1]}}
 {id=WavePlayer_0, properties {soundWaveAssetPtr=&/Plum_Absorb_Hit_01}}
 {id=WavePlayer_1, properties {soundWaveAssetPtr=&/PlumAbsorption_Absorb_01}}
 {id=WavePlayer_12, properties {soundWaveAssetPtr=&/PlumAbsorption_Absorb_06}}
 {id=WavePlayer_13, properties {soundWaveAssetPtr=&/PlumAbsorption_Absorb_07}}
 {id=WavePlayer_7, properties {soundWaveAssetPtr=&/PlumAbsorption_Absorb_05}}
]
properties {
 attenuationSettings=/Game/Sounds/Attenuation/Fort_Bullet_Impact_Athena_Att
 bOverrideConcurrency=true
 maxDistance=4500
 soundClassObject=&Fort_WeaponMelee
 volumeMultiplier=0.5
 concurrencyOverrides {bLimitToOwner=true, maxCount=1, resolutionRule=StopOldest}
}
