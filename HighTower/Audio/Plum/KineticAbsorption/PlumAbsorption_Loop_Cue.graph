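# Sound cue for the looping Plum kinetic absorption sound, classed as melee weapon audio with its own attenuation.
# A looping wave runs through an Enveloper that fades volume in over about half a second, then a Modulator at default settings.
# The Mixer at the root has a second, empty input.
# A concurrency override allows one instance per owner, keeping copies from stacking.
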SoundCue /HighTower/Audio/Plum/KineticAbsorption/PlumAbsorption_Loop_Cue
root=Mixer_0
nodes [
 {
  id=Enveloper_1
  children [WavePlayer_4]
  properties {
   pitchCurve {editorCurveData {keys [{time=0}]}}
   volumeCurve {editorCurveData {keys [
      {interpMode=RCIM_Cubic, tangentMode=RCTM_Break, time=0, value=0}
      {arriveTangent=2, interpMode=RCIM_Cubic, leaveTangent=2, tangentMode=RCTM_Break, time=0.1924, value=0.1502}
      <at 0.5 1>
     ]}}
  }
 }
 {id=Mixer_0, children [Modulator_0,null], properties {inputVolume [1,0.85]}}
 {id=Modulator_0, children [Enveloper_1]}
 {id=WavePlayer_4, properties {bLooping=true, soundWaveAssetPtr=&/PlumAbsorption_Loop_03}}
]
properties {
 attenuationSettings=&/PlumAbsorption_Att
 bOverrideConcurrency=true
 maxDistance=3584
 soundClassObject=&Fort_WeaponMelee
 volumeMultiplier=0.6
 concurrencyOverrides {bLimitToOwner=true, maxCount=1}
}
