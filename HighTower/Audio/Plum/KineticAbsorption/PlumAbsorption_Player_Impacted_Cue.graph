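# Sound cue for impacts on a player during Plum's kinetic absorption, as its name indicates.
# A random node picks one of seven impact wave variations with equal weight, then passes it through a modulator for variation.
# It plays in the melee weapon sound class with shared bullet-impact attenuation and its own concurrency override that caps simultaneous instances.
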SoundCue /HighTower/Audio/Plum/KineticAbsorption/PlumAbsorption_Player_Impacted_Cue
root=Modulator_0
nodes [
 {id=Modulator_0, children [Random_1]}
 {
  id=Random_1
  children [WavePlayer_0, WavePlayer_4, WavePlayer_5, WavePlayer_8, WavePlayer_9, WavePlayer_10, WavePlayer_11]
  properties {weights [1,1,1,1,1,1,1]}
 }
 {id=WavePlayer_0, properties {soundWaveAssetPtr=&/PlumAbsorption_Impact_01}}
 {id=WavePlayer_10, properties {soundWaveAssetPtr=&/PlumAbsorption_Impact_03}}
 {id=WavePlayer_11, properties {soundWaveAssetPtr=&/PlumAbsorption_Impact_02}}
 {id=WavePlayer_4, properties {soundWaveAssetPtr=&/PlumAbsorption_Impact_07}}
 {id=WavePlayer_5, properties {soundWaveAssetPtr=&/PlumAbsorption_Impact_06}}
 {id=WavePlayer_8, properties {soundWaveAssetPtr=&/PlumAbsorption_Impact_05}}
 {id=WavePlayer_9, properties {soundWaveAssetPtr=&/PlumAbsorption_Impact_04}}
]
properties {
 attenuationSettings=/Game/Sounds/Attenuation/Fort_Bullet_Impact_Athena_Att
 bOverrideConcurrency=true
 maxDistance=4500
 soundClassObject=&Fort_WeaponMelee
 volumeMultiplier=1.1
 concurrencyOverrides {maxCount=3}
}
