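# Sound cue for the start of Plum's kinetic absorption effect, layering a start sound with a build-up sound.
# A mixer plays both layers together, each drawing from three equally weighted random variants through its own modulator, with the start layer mixed lower and the build-up given slight pitch variation.
# It uses the Plum absorption attenuation, the melee weapon sound class and its own limit on simultaneous instances.
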SoundCue /HighTower/Audio/Plum/KineticAbsorption/PlumAbsorption_Start_Cue
root=Mixer_1
nodes [
 {id=Mixer_1, children [Modulator_1,Modulator_0], properties {inputVolume [0.7,1]}}
 {id=Modulator_0, children [Random_0]}
 {id=Modulator_1, children [Random_1], properties {pitchMax=1.02, pitchMin=0.98}}
 {id=Random_0, children [WavePlayer_1,WavePlayer_2,WavePlayer_3], properties {weights [1,1,1]}}
 {id=Random_1, children [WavePlayer_4,WavePlayer_5,WavePlayer_6], properties {weights [1,1,1]}}
 {id=WavePlayer_1, properties {soundWaveAssetPtr=&/PlumAbsorption_Start_01}}
 {id=WavePlayer_2, properties {soundWaveAssetPtr=&/PlumAbsorption_Start_03}}
 {id=WavePlayer_3, properties {soundWaveAssetPtr=&/PlumAbsorption_Start_02}}
 {id=WavePlayer_4, properties {soundWaveAssetPtr=&/PlumAbsorption_BuildUp_01}}
 {id=WavePlayer_5, properties {soundWaveAssetPtr=&/PlumAbsorption_BuildUp_02}}
 {id=WavePlayer_6, properties {soundWaveAssetPtr=&/PlumAbsorption_BuildUp_03}}
]
properties {
 attenuationSettings=&/PlumAbsorption_Att
 bOverrideConcurrency=true
 maxDistance=3584
 soundClassObject=&Fort_WeaponMelee
 volumeMultiplier=0.55
 concurrencyOverrides {maxCount=3}
}
