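# Sound cue for the drop sound of the Plum Kinetic Absorption item.
# A Random node picks one of two equally weighted Plum aura drop waves and feeds it through a Modulator at the root.
# It plays at reduced volume with weapon-drop attenuation and the melee weapon sound class.
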SoundCue /HighTower/Audio/Plum/KineticAbsorption/Plum_KineticAbsorption_Drop_Cue
root=Modulator_0
nodes [
 {id=Modulator_0, children [Random_1]}
 {id=Random_1, children [WavePlayer_0,WavePlayer_3], properties {weights [1,1]}}
 {id=WavePlayer_0, properties {soundWaveAssetPtr=&/Plum_Aura_Drop_01}}
 {id=WavePlayer_3, properties {soundWaveAssetPtr=&/Plum_Aura_Drop_02}}
]
properties {
 attenuationSettings=&Fort_Weapon_Drop_Att
 maxDistance=3000
 soundClassObject=&Fort_WeaponMelee
 volumeMultiplier=0.3
}
