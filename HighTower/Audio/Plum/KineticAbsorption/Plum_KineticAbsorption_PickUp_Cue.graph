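# Sound cue for the Plum Kinetic Absorption pickup, playing one of three Plum aura pickup sounds.
# An equally weighted Random node picks the wave and feeds it through a Modulator at the root.
# It plays at reduced volume in the melee weapon sound class with weapon pickup attenuation.
# A concurrency override prevents new instances from starting once the limit is reached.
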SoundCue /HighTower/Audio/Plum/KineticAbsorption/Plum_KineticAbsorption_PickUp_Cue
root=Modulator_0
nodes [
 {id=Modulator_0, children [Random_0]}
 {id=Random_0, children [WavePlayer_1,WavePlayer_2,WavePlayer_3], properties {weights [1,1,1]}}
 {id=WavePlayer_1, properties {soundWaveAssetPtr=&/Plum_Aura_PickUp_03}}
 {id=WavePlayer_2, properties {soundWaveAssetPtr=&/Plum_Aura_PickUp_01}}
 {id=WavePlayer_3, properties {soundWaveAssetPtr=&/Plum_Aura_PickUp_02}}
]
properties {
 attenuationSettings=&Fort_Weapon_Pickup_Att
 bOverrideConcurrency=true
 maxDistance=3500
 soundClassObject=&Fort_WeaponMelee
 volumeMultiplier=0.2
 concurrencyOverrides {resolutionRule=PreventNew}
}
